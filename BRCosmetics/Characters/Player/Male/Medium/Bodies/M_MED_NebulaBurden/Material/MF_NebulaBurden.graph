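# Material function for the NebulaBurden character skin, adding an animated galaxy effect and a glowing orb to incoming material attributes.
# The galaxy layer combines triplanar-projected nebula textures and stars with an emissive fresnel.
# The orb projects camera-facing plane UVs around an offset or parameter-driven center, layering core, ring, soft glow and twisted, noise-distorted radial swirls.
# TechArtMask channels supply orb and overall alphas, with results feeding emissive and a darkened base color.
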MaterialFunction /BRCosmetics/Characters/Player/Male/Medium/Bodies/M_MED_NebulaBurden/Material/MF_NebulaBurden
graph {nodes [
  FunctionOutput(None=BlendMaterialAttributes(A=NamedRerouteUsage(), B=NamedRerouteUsage(), Alpha=NamedRerouteUsage()))
  Multiply(A=VectorParameter() {group=Orb, parameterName=OrbSoftGlowColor, defaultValue=<rgba 0.0387 0.521 1 1>}.RGB, B=NamedRerouteUsage())
  n1 = FunctionInput() {bUsePreviewValueAsDefault=true, inputName=TechArtMask, sortPriority=1, previewValue=<1 1 1 1>}
  n2 = NamedRerouteDeclaration(None=@n1) {name=TechArtMask, nodeColor=<rgba 0 1 0.6953 1>}
  NamedRerouteDeclaration(None=ComponentMask(None=@n2) {r=true}) {name=OrbAlpha, nodeColor=<rgba 0.7109 1 0 1>}
  n3 = NamedRerouteUsage()
  n4 = Constant() {r=0.5}
  n5 = ScalarParameter() {defaultValue=0.15, group=Orb, parameterName=OrbSoftGlowRadius}
  NamedRerouteDeclaration(None=Multiply(
    A=SphereMask(A=@n3, B=@n4, Radius=@n5) {hardnessPercent=0}
    B=ScalarParameter() {defaultValue=0.4, group=Orb, parameterName=OrbSoftGlowStrength}
   )) {name=OrbSoftGlow, nodeColor=<rgba 0.9219 1 0 1>}
  n6 = NamedRerouteUsage()
  n7 = CameraPositionWS()
  n8 = NamedRerouteDeclaration(None=Subtract(A=NamedRerouteUsage(), B=NamedRerouteDeclaration(None=@n7) {name=CamPos, nodeColor=<rgba 1 0.9844 0 1>})) {name=CamToCenter, nodeColor=<rgba 1 0 0.75 1>}
  n9 = NamedRerouteUsage()
  n10 = CameraVectorWS()
  n11 = Multiply(A=Divide(A=DotProduct(A=@n8, B=@n9), B=DotProduct(A=@n9, B=@n10)), B=@n10)
  NamedRerouteDeclaration(None=Subtract(A=Add(A=@n6, B=@n11), B=NamedRerouteUsage())) {name=CenteredWSP, nodeColor=<rgba 0.4297 1 0 1>}
  n12 = Add(A=NamedRerouteUsage(), B=NamedRerouteUsage())
  n13 = Multiply(A=NamedRerouteUsage(), B=ScalarParameter() {defaultValue=2, group=Orb, parameterName=OrbCoreVerticalOffset})
  n14 = NamedRerouteDeclaration(None=ScalarParameter() {defaultValue=8, group=Orb, parameterName=OrbCoreRadius}) {name=Radius, nodeColor=<rgba 0.0312 0 1 1>}
  n15 = ScalarParameter() {defaultValue=0.1, group=Orb, parameterName=OrbCoreHardness}
  NamedRerouteDeclaration(None=Power(
    Base=SphereMask(A=@n12, B=@n13, Radius=@n14, Hardness=@n15) {attenuationRadius=5, hardnessPercent=10}
    Exp=ScalarParameter() {defaultValue=4, group=Orb, parameterName=OrbCoreExponent}
   )) {name=OrbCore, nodeColor=<rgba 0 0.3906 1 1>}
  n16 = OneMinus(None=Saturate(None=NamedRerouteUsage()))
  n17 = ScalarParameter() {defaultValue=6, group=Orb, parameterName=OrbCoreDistortionStrength2}
  n18 = NamedRerouteUsage()
  n19 = Saturate(None=Divide(A=Length(None=@n18) {is=class MaterialExpressionLength}, B=NamedRerouteUsage()) {constB=5})
  NamedRerouteDeclaration(None=Multiply(A=Multiply(A=@n16, B=@n17), B=Power(Base=@n19) {constExponent=0.8})) {name=CoreDistortion, nodeColor=<rgba 0 0.9766 1 1>}
  n20 = NamedRerouteDeclaration(None=CameraVectorWS()) {name=PlaneNormal, nodeColor=<rgba 1 0 0.1172 1>}
  n21 = Transform(None=Constant3Vector() {constant=<rgba 0 1 0 0>}) {transformSourceType=TRANSFORMSOURCE_View}
  NamedRerouteDeclaration(None=Normalize(VectorInput=CrossProduct(A=@n20, B=@n21))) {name=PlaneXAxis, nodeColor=<rgba 0.7812 0 1 1>}
  n22 = NamedRerouteUsage()
  n23 = ScalarParameter() {defaultValue=0.04, group=Orb, parameterName=OrbSwirl_03_DistortionStrength}
  n24 = ScalarParameter() {defaultValue=1.3, group=Orb, parameterName=OrbSwirl_03_Twist}
  n25 = NamedRerouteUsage()
  n26 = MaterialFunctionCall(Float2=NamedRerouteUsage()) {materialFunction=&BreakOutFloat2Components}
  n27 = ScalarParameter() {defaultValue=0.05, group=Orb, parameterName=OrbSwirl_03_Speed}
  n28 = Time()
  n29 = MaterialFunctionCall(X=Add(A=Multiply(A=@n24, B=@n25), B=@n26.R), Y=Add(A=@n26.G, B=Multiply(A=@n27, B=@n28))) {materialFunction=&MakeFloat2}
  n30 = ScalarParameter() {defaultValue=1, group=Orb, parameterName=OrbSwirl_03_Tile}
  n31 = TextureCoordinate()
  n32 = TextureSample!(
   UVs=Add(A=Multiply(A=@n22, B=@n23), B=Multiply(A=@n29.Result, B=@n30))
   Tex
   "DDX(UVs)"=DDX(Value=@n31)
   "DDY(UVs)"=DDY(Value=@n31)
   Apply View MipBias
  ) {mipValueMode=TMVM_Derivative, texture=/BRCosmetics/Characters/Player/Male/Medium/Bodies/M_MED_NebulaBurden/Texture/T_NebulaBurden_OrbNoisev2}
  n33 = ScalarParameter() {defaultValue=2, group=Orb, parameterName=OrbStreaks2Exponent}
  NamedRerouteDeclaration(None=Multiply(A=Power(Base=@n32.R, Exp=@n33), B=NamedRerouteUsage())) {name=OrbStreaks2, nodeColor=<rgba 1 0.9609 0 1>}
  n34 = TextureObjectParameter() {
   group=Galaxy
   parameterName=StarsTexture
   samplerType=SAMPLERTYPE_Masks
   sortPriority=80
   texture=/Game/Characters/Player/Male/Medium/Bodies/M_Med_Soldier_04/Skins/Galaxy/Textures/Celestial_T_Stars
  }
  n35 = VectorParameter() {group=Galaxy, parameterName=Stars_RotationSpeed, sortPriority=84, defaultValue=<rgba 0.1 0.1 0.1 1>}
  n36 = MaterialFunctionCall!(
   Texture=@n34
   Depth=ScalarParameter() {defaultValue=150, group=Galaxy, parameterName=Stars_Depth, sortPriority=81}
   Tile=ScalarParameter() {defaultValue=2, group=Galaxy, parameterName=Stars_Tile, sortPriority=82}
   BlendContrast=ScalarParameter() {defaultValue=1, group=Galaxy, parameterName=Stars_BlendContrast, sortPriority=83}
   SpeedXY=ComponentMask(None=VectorParameter() {group=Galaxy, parameterName=Stars_Speed, sortPriority=84, defaultValue=<rgba 0.002 -0.03 0 1>}.RGB) {g=true, r=true}
   RotationSpeed=@n35.RGB
   ActorOffset=NamedRerouteDeclaration(None=VectorParameter() {group=Galaxy, parameterName=GalaxyOffset, defaultValue=<rgba 0 0 80 1>}.RGB) {name=ActorOffset, nodeColor=<rgba 1 0.0469 0 1>}
  ) {materialFunction=&/MF_NebulaBurdenTriplanarBox}
  NamedRerouteDeclaration(None=@n36.UVX) {name=UVX, nodeColor=<rgba 0 0.8828 1 1>}
  NamedRerouteDeclaration(None=@n36.UVY) {name=UVY, nodeColor=<rgba 0 0.8828 1 1>}
  NamedRerouteDeclaration(None=@n36.UVZ) {name=UVZ, nodeColor=<rgba 0 0.8828 1 1>}
  NamedRerouteDeclaration(None=@n36.BlendY) {name=BlendY, nodeColor=<rgba 1 0.6226 0.1707 1>}
  NamedRerouteDeclaration(None=@n36.BlendZ) {name=BlendZ, nodeColor=<rgba 1 0.6226 0.1707 1>}
  n37 = TextureObjectParameter() {
   group=Galaxy
   parameterName=Galaxy_01_Texture
   sortPriority=58
   texture=/Game/Characters/Player/Male/Medium/Bodies/M_Med_Soldier_04/Skins/Galaxy/Textures/T_FN_Nebula_style_blue
  }
  NamedRerouteDeclaration(None=@n37) {name=GalaxyTexture, nodeColor=<rgba 0 1 0.9062 1>}
  NamedRerouteDeclaration(None=Multiply(A=ScalarParameter() {defaultValue=0.6, group=Galaxy, parameterName=Galaxy_01_Tile, sortPriority=62}) {constB=0.01}) {name=GalaxyTile, nodeColor=<rgba 1 0 0.1172 1>}
  n38 = VectorParameter() {group=Galaxy, parameterName=Galaxy_01_Speed, sortPriority=64, defaultValue=<rgba 0.0035 -0.005 0 1>}
  NamedRerouteDeclaration(None=Multiply(A=ComponentMask(None=@n38.RGB) {g=true, r=true}, B=Time())) {name=GalaxySpeed, nodeColor=<rgba 0 0.8125 1 1>}
  n39 = VectorParameter() {group=Orb, parameterName=OuterRingColor, defaultValue=<rgba 0 0.61 1 1>}
  n40 = ScalarParameter() {defaultValue=2, group=Orb, parameterName=OrbOuterRingStrength}
  NamedRerouteDeclaration(None=Multiply(A=Multiply(A=@n39.RGB, B=@n40), B=NamedRerouteUsage())) {name=OuterRingFinal, nodeColor=<rgba 0.5703 0 1 1>}
  n41 = VectorParameter() {group=Orb, parameterName=StreaksColor, defaultValue=<rgba 0.0962 0.0144 0.5052 1>}
  n42 = ScalarParameter() {defaultValue=15, group=Orb, parameterName=OrbSwirl_02_Strength}
  NamedRerouteDeclaration(None=Multiply(A=Multiply(A=@n41.RGB, B=@n42), B=NamedRerouteUsage())) {name=OrbSwirl2Final, nodeColor=<rgba 0 0.5078 1 1>}
  NamedRerouteDeclaration(None=Multiply(A=VectorParameter() {group=Orb, parameterName=OrbSwirlsColor, defaultValue=<rgba 0 0.4513 0.849 1>}.RGB, B=NamedRerouteUsage())) {name=OrbSwirls1Final, nodeColor=<rgba 0.5703 1 0 1>}
  n43 = VectorParameter() {group=Orb, parameterName=OrbCoreColor, defaultValue=<rgba 0 0.6983 1 1>}
  n44 = ScalarParameter() {defaultValue=20, group=Orb, parameterName=OrbCoreStrength}
  NamedRerouteDeclaration(None=Multiply(A=Multiply(A=@n43.RGB, B=@n44), B=NamedRerouteUsage())) {name=OrbCoreFinal, nodeColor=<rgba 0.2422 0 1 1>}
  n45 = VectorParameter() {group=Orb, parameterName=OrbStreaks2Color, defaultValue=<rgba 0.6406 0.0712 0.1673 1>}
  n46 = ScalarParameter() {defaultValue=5, group=Orb, parameterName=OrbSwirl_03_Strength}
  NamedRerouteDeclaration(None=Multiply(A=NamedRerouteUsage(), B=Multiply(A=@n45.RGB, B=@n46))) {name=OrbStreaks2Final, nodeColor=<rgba 1 0.0469 0 1>}
  n47 = Multiply(
   A=Fresnel(ExponentIn=ScalarParameter() {defaultValue=3, group=Galaxy, parameterName=EmissiveFresnelExponent, sortPriority=90}) {baseReflectFraction=0}
   B=ScalarParameter() {defaultValue=1, group=Galaxy, parameterName=EmissiveFresnelIntensity, sortPriority=92}
  )
  NamedRerouteDeclaration(None=Multiply(
    A=NamedRerouteDeclaration(None=@n47) {name=FresnelMask, nodeColor=<rgba 0 0.7891 1 1>}
    B=VectorParameter() {group=Galaxy, parameterName=EmissiveColor, sortPriority=91, defaultValue=<rgba 0.0885 0.6921 1 1>}.RGB
   )) {name=Fresnel, nodeColor=<rgba 0.8047 0 1 1>}
  n48 = Multiply(A=NamedRerouteUsage(), B=NamedRerouteUsage())
  n49 = NamedRerouteUsage()
  n50 = Multiply(A=NamedRerouteUsage(), B=NamedRerouteUsage())
  n51 = NamedRerouteUsage()
  n52 = LinearInterpolate(
   A=TextureSample!(UVs=Add(A=@n48, B=@n49), Tex=NamedRerouteUsage(), Apply View MipBias) {samplerSource=SSM_Wrap_WorldGroupSettings}.RGB
   B=TextureSample(UVs=Add(A=@n50, B=@n51), Tex=NamedRerouteUsage()) {samplerSource=SSM_Wrap_WorldGroupSettings}.RGB
   Alpha=NamedRerouteUsage()
  )
  n53 = NamedRerouteUsage()
  n54 = NamedRerouteUsage()
  n55 = TextureSample(UVs=Add(A=Multiply(A=@n53, B=@n54), B=NamedRerouteUsage()), Tex=NamedRerouteUsage()) {samplerSource=SSM_Wrap_WorldGroupSettings}
  n56 = NamedRerouteUsage()
  NamedRerouteDeclaration(None=Power(
    Base=LinearInterpolate(A=@n52, B=@n55.RGB, Alpha=@n56)
    Exp=ScalarParameter() {defaultValue=1.5, group=Galaxy, parameterName=Galaxy_01_Exp, sortPriority=66}
   )) {name=Galaxies, nodeColor=<rgba 1 0 0.2578 1>}
  n57 = OneMinus(None=NamedRerouteUsage())
  n58 = NamedRerouteUsage()
  NamedRerouteDeclaration(None=Add(A=Multiply(A=@n57, B=@n58), B=NamedRerouteUsage())) {name=GalaxyEmissive, nodeColor=<rgba 1 0 0.1875 1>}
  n59 = GetMaterialAttributes(None=FunctionInput() {inputName=Material, inputType=FunctionInput_MaterialAttributes}) {attributeGetTypes [#1, #2]}
  n60 = NamedRerouteDeclaration(None=@n59.EmissiveColor) {name=IN_Emissive, nodeColor=<rgba 1 0 0.9844 1>}
  n61 = NamedRerouteUsage()
  n62 = ScalarParameter() {defaultValue=0.3, group=Galaxy, parameterName=UseBaseColor}
  NamedRerouteDeclaration(None=SetMaterialAttributes!(
    MaterialAttributes=NamedRerouteDeclaration(None=@n59.MaterialAttributes) {name=IN_Material, nodeColor=<rgba 1 0 0.9609 1>}
    EmissiveColor=Add(A=@n60, B=@n61)
    BaseColor=LinearInterpolate(B=@n59.BaseColor, Alpha=@n62)
   ) {attributeSetTypes [#2, #1]}) {name=MaterialOUT, nodeColor=<rgba 0.0078 1 0 1>}
  NamedRerouteDeclaration(None=ComponentMask(None=@n2) {b=true}) {name=Alpha, nodeColor=<rgba 0.7109 1 0 1>}
  n63 = ChannelMaskParameter(None=NamedRerouteUsage()) {group=Galaxy, maskChannel=Blue, parameterName=StarMaskChannel, sortPriority=84, defaultValue=<rgba 0 0 1 0>}
  NamedRerouteDeclaration(None=Multiply(A=@n63.R, B=ScalarParameter() {defaultValue=3, group=Galaxy, parameterName=GalaxyStarMaskStrength, sortPriority=85}) {constB=2}) {name=StarMask, nodeColor=<rgba 0.9922 1 0 1>}
  n64 = LinearInterpolate(A=NamedRerouteUsage(), Alpha=NamedRerouteUsage()) {constB=0}
  NamedRerouteDeclaration(None=Add(A=Saturate(None=@n64), B=NamedRerouteUsage())) {name=GalaxiesAndStars, nodeColor=<rgba 0 1 0.9062 1>}
  n65 = ComponentMask(None=@n36.Result) {b=true, r=true}
  n66 = Multiply(A=@n65, B=@n65)
  NamedRerouteDeclaration(None=Multiply(
    A=MaterialFunctionCall(f2=@n66) {materialFunction=&AddComponents}.f2
    B=ScalarParameter() {defaultValue=1, group=Galaxy, parameterName=Stars_Brightness, sortPriority=84}
   )) {name=Stars, nodeColor=<rgba 0 1 0.5078 1>}
  n67 = Add(B=ScalarParameter() {defaultValue=0.02, group=Orb, parameterName=OrbOuterRingOffset}) {constA=0.5}
  n68 = Add(
   A=Multiply(A=NamedRerouteUsage(), B=ScalarParameter() {defaultValue=0.03, group=Orb, parameterName=OrbOuterRingDistortionStrength})
   B=ScalarParameter() {defaultValue=0.04, group=Orb, parameterName=OrbOuterRingThickness}
  )
  n69 = Saturate(None=SphereMask(A=NamedRerouteUsage(), B=Subtract(A=@n67, B=@n68) {constA=0.5}, Radius=@n68) {attenuationRadius=0.1, hardnessPercent=0})
  n70 = ScalarParameter() {defaultValue=4, group=Orb, parameterName=OrbOuterRingExponent}
  NamedRerouteDeclaration(None=Multiply(A=Power(Base=@n69, Exp=@n70), B=NamedRerouteUsage())) {name=OuterRing, nodeColor=<rgba 0 0.7891 1 1>}
  n71 = NamedRerouteUsage()
  n72 = CrossProduct(A=NamedRerouteUsage(), B=NamedRerouteUsage())
  n73 = AppendVector(A=DotProduct(A=NamedRerouteUsage(), B=@n71), B=DotProduct(A=@n71, B=Normalize(VectorInput=@n72)))
  n74 = ScalarParameter() {defaultValue=28, group=Orb, parameterName=OrbUVDivisor}
  NamedRerouteDeclaration(None=Add(A=Divide(A=@n73, B=@n74)) {constB=0.5}) {name=OrbUVs, nodeColor=<rgba 1 0.7734 0 1>}
  n75 = MaterialFunctionCall(Float2=NamedRerouteUsage()) {materialFunction=&BreakOutFloat2Components}
  n76 = Multiply(A=Time(), B=ScalarParameter() {defaultValue=0.06, group=Orb, parameterName=OrbNoiseSpeed})
  n77 = TextureSample(UVs=MaterialFunctionCall(X=@n75.R, Y=Add(A=@n75.G, B=@n76)) {materialFunction=&MakeFloat2}.Result) {
   samplerType=SAMPLERTYPE_LinearGrayscale
   texture=/Game/Global/Materials/MaterialGym/Materials/SmokeWithFlowmap/Textures/T_Swirly_Noise_MiddleDetail
  }
  NamedRerouteDeclaration(None=ConstantBiasScale(None=NamedRerouteDeclaration(None=@n77.R) {name=OrbNoise0-1, nodeColor=<rgba 1 0.5625 0 1>}) {bias=-0.5, scale=2}) {name=OrbNoise, nodeColor=<rgba 0 1 0.4844 1>}
  n78 = VectorParameter() {group=Orb, parameterName=OrbNoiseMaskSpeed, defaultValue=<rgba 0 -0.02 0 1>}
  n79 = TextureSample(UVs=Panner(Coordinate=NamedRerouteUsage(), Speed=ComponentMask(None=@n78.RGB) {g=true, r=true})) {
   samplerType=SAMPLERTYPE_LinearGrayscale
   texture=&T_Swirly_Noise_MiddleDetail
  }
  NamedRerouteDeclaration(None=@n79.R) {name=OrbNoiseMask, nodeColor=<rgba 0.8516 1 0 1>}
  n80 = NamedRerouteUsage()
  n81 = ScalarParameter() {defaultValue=0.02, group=Orb, parameterName=OrbSwirlDistortStrength}
  n82 = Subtract(A=NamedRerouteUsage()) {constB=0.5}
  n83 = MaterialFunctionCall(Vector 2=@n82) {materialFunction=/Engine/Functions/Engine_MaterialFunctions02/Utility/VectorLength}
  n84 = Multiply(
   A=ScalarParameter() {defaultValue=1, group=Orb, parameterName=OrbSwirlsTwist}
   B=Power(
    Base=NamedRerouteDeclaration(None=@n83.V2 Length) {name=OrbLength, nodeColor=<rgba 0 0.6953 1 1>}
    Exp=ScalarParameter() {defaultValue=0.5, group=Orb, parameterName=OrbSwirlTwistExponent}
   )
  )
  n85 = MaterialFunctionCall(Float2=NamedRerouteDeclaration(None=MaterialFunctionCall(Vector or UVs=@n82) {materialFunction=&VectorToRadialValue}.Radial Coordinates) {name=OrbRadialUVs, nodeColor=<rgba 0.8047 1 0 1>}) {materialFunction=&BreakOutFloat2Components}
  n86 = Multiply(A=ScalarParameter() {defaultValue=0.05, group=Orb, parameterName=OrbSwirlsSpeed}, B=Time())
  n87 = Multiply(
   A=MaterialFunctionCall(X=Add(A=@n84, B=@n85.R), Y=Add(A=@n86, B=@n85.G)) {materialFunction=&MakeFloat2}.Result
   B=ScalarParameter() {defaultValue=1, group=Orb, parameterName=OrbSwirlsTile}
  )
  n88 = TextureCoordinate()
  n89 = TextureSample!(
   UVs=Add(A=Multiply(A=@n80, B=@n81), B=NamedRerouteDeclaration(None=@n87) {name=OrbSwirlUV, nodeColor=<rgba 0 1 0.0391 1>})
   Tex
   "DDX(UVs)"=DDX(Value=@n88)
   "DDY(UVs)"=DDY(Value=@n88)
   Apply View MipBias
  ) {mipValueMode=TMVM_Derivative, texture=/BRCosmetics/Characters/Player/Male/Medium/Bodies/M_MED_NebulaBurden/Texture/T_NebulaBurden_OrbNoisev4}
  NamedRerouteDeclaration(None=@n89.R) {name=OrbSwirlSample, nodeColor=<rgba 0.0781 0 1 1>}
  n90 = NamedRerouteUsage()
  n91 = ScalarParameter() {defaultValue=0.02, group=Orb, parameterName=OrbSwirl_02_DistortionStrength}
  n92 = ScalarParameter() {defaultValue=0.8, group=Orb, parameterName=OrbSwirl_02_Twist}
  n93 = NamedRerouteUsage()
  n94 = MaterialFunctionCall(Float2=NamedRerouteUsage()) {materialFunction=&BreakOutFloat2Components}
  n95 = ScalarParameter() {defaultValue=0.1, group=Orb, parameterName=OrbSwirl_02_Speed}
  n96 = Time()
  n97 = MaterialFunctionCall(X=Add(A=Multiply(A=@n92, B=@n93), B=@n94.R), Y=Add(A=@n94.G, B=Multiply(A=@n95, B=@n96))) {materialFunction=&MakeFloat2}
  n98 = ScalarParameter() {defaultValue=1, group=Orb, parameterName=OrbSwirl_02_Tile}
  n99 = TextureCoordinate()
  n100 = TextureSample(UVs=Add(A=Multiply(A=@n90, B=@n91), B=Multiply(A=@n97.Result, B=@n98)), "DDX(UVs)"=DDX(Value=@n99), "DDY(UVs)"=DDY(Value=@n99)) {mipValueMode=TMVM_Derivative, texture=&T_NebulaBurden_OrbNoisev4}
  n101 = ScalarParameter() {defaultValue=2.5, group=Orb, parameterName=OrbSwirl_02_Exponent}
  NamedRerouteDeclaration(None=Multiply(A=Power(Base=@n100.R, Exp=@n101), B=NamedRerouteUsage())) {name=OrbSwirl_02, nodeColor=<rgba 1 0.9609 0 1>}
  n102 = NamedRerouteUsage()
  NamedRerouteDeclaration(None=OneMinus(None=SmoothStep(Value=@n102) {constMax=0.6, constMin=0.4})) {name=OrbMask, nodeColor=<rgba 0.3125 1 0 1>}
  n103 = TransformPosition(None=Constant3Vector()) {transformType=TRANSFORMPOSSOURCE_World}
  n104 = VectorParameter() {group=Orb, parameterName=OrbOffset, defaultValue=<rgba 0 -5 128 1>}
  n105 = VectorParameter() {group=Orb, parameterName=OrbLocation, defaultValue=<rgba 0 0 0 1>}
  n106 = VectorParameter() {group=Orb, parameterName=OrbUpAxis, defaultValue=<rgba 0 0 1 1>}
  n107 = Add(
   A=Multiply(
    A=NamedRerouteDeclaration(None=@n106.RGB) {name=OrbUpVector, nodeColor=<rgba 0 0.8125 1 1>}
    B=ScalarParameter() {defaultValue=-3, group=Orb, parameterName=OrbOffsetUpRuntime}
   )
   B=Multiply(
    A=ScalarParameter() {defaultValue=-2, group=Orb, parameterName=OrbOffsetForwardRuntime}
    B=VectorParameter() {group=Orb, parameterName=OrbForwardAxis, defaultValue=<rgba 0 1 0 1>}.RGB
   )
  )
  NamedRerouteDeclaration(None=LinearInterpolate(A=Add(A=@n103, B=@n104.RGB), B=Add(A=@n105.RGB, B=@n107), Alpha=ScalarParameter() {group=Orb, parameterName=UseCCPMOrbLocation})) {name=OrbCenter, nodeColor=<rgba 1 0.6797 0 1>}
  n108 = Multiply(
   A=Power(Base=NamedRerouteUsage(), Exp=ScalarParameter() {defaultValue=3, group=Orb, parameterName=OrbSwirl_01_Exponent})
   B=NamedRerouteUsage()
  )
  n109 = LinearInterpolate(B=NamedRerouteUsage(), Alpha=ScalarParameter() {defaultValue=0.5, group=Orb, parameterName=OrbSwirl_01_NoiseStrength}) {constA=1}
  NamedRerouteDeclaration(None=Multiply(A=Multiply(A=@n108, B=@n109), B=ScalarParameter() {defaultValue=15, group=Orb, parameterName=OrbSwirl_01_Strength})) {name=OrbSwirl_01, nodeColor=<rgba 0 0.3672 1 1>}
  n110 = NamedRerouteUsage()
  n111 = Add(A=NamedRerouteUsage(), B=NamedRerouteUsage())
  n112 = NamedRerouteUsage()
  n113 = Add(A=Add(A=Add(A=@n111, B=@n112), B=NamedRerouteUsage()), B=NamedRerouteUsage())
  NamedRerouteDeclaration(None=Add(A=NamedRerouteUsage(), B=Multiply(A=@n110, B=@n113))) {name=FinalEmissive, nodeColor=<rgba 0 1 0.1094 1>}
 ]}
