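# Physics asset for the NebulaBurden medium male player body cosmetic, covering the lower body and its dynamic belt and skirt bones.
# Capsules on pelvis, legs and feet accompany two-segment skirt chains, each rooted in a small MSTR body.
# Skirt segments have tight twist and secondary swing limits, with wider primary swing.
# Front MSTRs link through free-rotation belt bodies to the pelvis; among the other MSTRs, only left-back has a pelvis constraint.
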PhysicsAsset /BRCosmetics/Characters/Player/Male/Medium/Bodies/M_MED_NebulaBurden/Meshes/M_MED_NebulaBurden_Physics
bodies {
 calf_l [{shapeType=Capsule, length=24, radius=8, shapeName=calf_l_capsule, center=<-20 -3 0>, rotation=<rot 0 85 90>}]
 calf_r [{shapeType=Capsule, length=24, radius=8, shapeName=calf_r_capsule, center=<20 3 0>, rotation=<rot 0 85 90>}]
 dyn_belt_l [{shapeType=Capsule, length=0, radius=1, shapeName=dyn_belt_l_capsule}]
 dyn_belt_mid [{shapeType=Capsule, length=0, radius=1, shapeName=dyn_belt_mid_capsule, center=<-0.293 0 0>}]
 dyn_belt_r [{shapeType=Capsule, length=0, radius=1, shapeName=dyn_belt_r_capsule, center=<-0.293 0 0>}]
 dyn_skirt_back_l [{shapeType=Capsule, length=8, radius=8, shapeName=dyn_skirt_back_l_capsule, center=<8 -5 0>, rotation=<rot 0 90 90>}]
 dyn_skirt_back_l_01 [{shapeType=Capsule, length=8, radius=8, shapeName=dyn_skirt_back_l_01_capsule, center=<8 -5 0>, rotation=<rot 0 90 90>}]
 dyn_skirt_back_l_MSTR [{shapeType=Capsule, length=0.505, radius=0.505, shapeName=dyn_skirt_back_l_MSTR_capsule, center=<0 0.293 0>, rotation=<rot 0 -90 0>}]
 dyn_skirt_back_r [{shapeType=Capsule, length=8, radius=8, shapeName=dyn_skirt_back_r_capsule, center=<8 -5 0>, rotation=<rot 0 90 90>}]
 dyn_skirt_back_r_01 [{shapeType=Capsule, length=8, radius=8, shapeName=dyn_skirt_back_r_01_capsule, center=<8 -5 0>, rotation=<rot 0 90 90>}]
 dyn_skirt_back_r_MSTR [{shapeType=Capsule, length=0.505, radius=0.505, shapeName=dyn_skirt_back_r_MSTR_capsule, center=<0 0.293 0>, rotation=<rot 0 -90 0>}]
 dyn_skirt_front_MSTR [{shapeType=Capsule, length=0.505, radius=0.505, shapeName=dyn_skirt_front_MSTR_capsule, center=<0 0.293 0>, rotation=<rot 0 -90 0>}]
 dyn_skirt_front_f [{shapeType=Capsule, length=8, radius=3, shapeName=dyn_skirt_front_f_capsule, center=<8 3 0>, rotation=<rot 0 90 90>}]
 dyn_skirt_front_f1 [{shapeType=Capsule, length=8, radius=3, shapeName=dyn_skirt_front_f1_capsule, center=<8 3 0>, rotation=<rot 0 90 90>}]
 dyn_skirt_front_l_01_01 [{shapeType=Capsule, length=8, radius=3, shapeName=dyn_skirt_front_l_01_01_capsule, center=<8 3 0>, rotation=<rot 0 90 90>}]
 dyn_skirt_front_l_01_02 [{shapeType=Capsule, length=8, radius=3, shapeName=dyn_skirt_front_l_01_02_capsule, center=<8 3 0>, rotation=<rot 0 90 90>}]
 dyn_skirt_front_l_01_MSTR [{shapeType=Capsule, length=0.505, radius=0.505, shapeName=dyn_skirt_front_l_01_MSTR_capsule, rotation=<rot 0 -90 0>}]
 dyn_skirt_front_l_02_01 [{shapeType=Capsule, length=8, radius=3, shapeName=dyn_skirt_front_l_02_01_capsule, center=<8 3 0>, rotation=<rot 0 90 90>}]
 dyn_skirt_front_l_02_02 [{shapeType=Capsule, length=8, radius=3, shapeName=dyn_skirt_front_l_02_02_capsule, center=<8 3 0>, rotation=<rot 0 90 90>}]
 dyn_skirt_front_l_02_MSTR [{shapeType=Capsule, length=0.505, radius=0.505, shapeName=dyn_skirt_front_l_02_MSTR_capsule, center=<0 0.293 0>, rotation=<rot 0 -90 0>}]
 dyn_skirt_front_r_01_01 [{shapeType=Capsule, length=8, radius=3, shapeName=dyn_skirt_front_r_01_01_capsule, center=<8 3 0>, rotation=<rot 0 90 90>}]
 dyn_skirt_front_r_01_02 [{shapeType=Capsule, length=8, radius=3, shapeName=dyn_skirt_front_r_01_02_capsule, center=<8 3 0>, rotation=<rot 0 90 90>}]
 dyn_skirt_front_r_01_MSTR [{shapeType=Capsule, length=0.505, radius=0.505, shapeName=dyn_skirt_front_r_01_MSTR_capsule, center=<0 0.293 0>, rotation=<rot 0 -90 0>}]
 dyn_skirt_front_r_02_01 [{shapeType=Capsule, length=8, radius=3, shapeName=dyn_skirt_front_r_02_01_capsule, center=<8 3 0>, rotation=<rot 0 90 90>}]
 dyn_skirt_front_r_02_02 [{shapeType=Capsule, length=8, radius=3, shapeName=dyn_skirt_front_r_02_02_capsule, center=<8 3 0>, rotation=<rot 0 90 90>}]
 dyn_skirt_front_r_02_MSTR [{shapeType=Capsule, length=0.505, radius=0.505, shapeName=dyn_skirt_front_r_02_MSTR_capsule, center=<0 0.293 0>, rotation=<rot 0 -90 0>}]
 dyn_skirt_l_back [{shapeType=Capsule, length=8, radius=8, shapeName=dyn_skirt_l_back_capsule, center=<8 6 0>, rotation=<rot 0 90 90>}]
 dyn_skirt_l_back_01 [{shapeType=Capsule, length=8, radius=8, shapeName=dyn_skirt_l_back_01_capsule, center=<8 6 0>, rotation=<rot 0 90 90>}]
 dyn_skirt_l_back_MSTR [{shapeType=Capsule, length=0.505, radius=0.505, shapeName=dyn_skirt_l_back_MSTR_capsule, rotation=<rot 0 -90 0>}]
 dyn_skirt_l_fr [{shapeType=Capsule, length=8, radius=4.5, shapeName=dyn_skirt_l_fr_capsule, center=<8 3 3>, rotation=<rot 0 90 90>}]
 dyn_skirt_l_fr_01 [{shapeType=Capsule, length=8, radius=4.5, shapeName=dyn_skirt_l_fr_01_capsule, center=<8 3 3>, rotation=<rot 0 90 90>}]
 dyn_skirt_l_fr_MSTR [{shapeType=Capsule, length=0.505, radius=0.505, shapeName=dyn_skirt_l_fr_MSTR_capsule, center=<0 0.293 0>, rotation=<rot 0 -90 0>}]
 dyn_skirt_r_back [{shapeType=Capsule, length=8, radius=8, shapeName=dyn_skirt_r_back_capsule, center=<8 -6 0>, rotation=<rot 0 90 90>}]
 dyn_skirt_r_back_01 [{shapeType=Capsule, length=8, radius=8, shapeName=dyn_skirt_r_back_01_capsule, center=<8 -6 0>, rotation=<rot 0 90 90>}]
 dyn_skirt_r_back_MSTR [{shapeType=Capsule, length=0.505, radius=0.505, shapeName=dyn_skirt_r_back_MSTR_capsule, center=<0 0.293 0>, rotation=<rot 0 -90 0>}]
 dyn_skirt_r_fr [{shapeType=Capsule, length=8, radius=4.5, shapeName=dyn_skirt_r_fr_capsule, center=<8 -3 3>, rotation=<rot 0 90 90>}]
 dyn_skirt_r_fr_01 [{shapeType=Capsule, length=8, radius=4.5, shapeName=dyn_skirt_r_fr_01_capsule, center=<8 -3 3>, rotation=<rot 0 90 90>}]
 dyn_skirt_r_fr_MSTR [{shapeType=Capsule, length=0.505, radius=0.505, shapeName=dyn_skirt_r_fr_MSTR_capsule, center=<0 0.293 0>, rotation=<rot 0 -90 0>}]
 foot_l [{shapeType=Capsule, length=14, radius=8, shapeName=foot_l_capsule, center=<-4 5 0>, rotation=<rot 0 0 90>}]
 foot_r [{shapeType=Capsule, length=14, radius=8, shapeName=foot_r_capsule, center=<4 -5 0>, rotation=<rot 0 0 90>}]
 pelvis [{shapeType=Capsule, length=20, radius=7.4, shapeName=pelvis_capsule}]
 thigh_l [{shapeType=Capsule, length=26, radius=8, shapeName=thigh_l_capsule, center=<-21 -1 0>, rotation=<rot 90 0 0>}]
 thigh_r [{shapeType=Capsule, length=26, radius=8, shapeName=thigh_l_capsule, center=<21 1 0>, rotation=<rot 90 0 180>}]
}
constraints [
 {bone1Name=dyn_skirt_front_l_01_01, bone2Name=dyn_skirt_front_l_01_MSTR, swing1LimitDegrees=6, swing2LimitDegrees=2, twistLimitDegrees=2}
 {bone1Name=dyn_skirt_front_l_01_02, bone2Name=dyn_skirt_front_l_01_01, swing1LimitDegrees=6, swing2LimitDegrees=2, twistLimitDegrees=2}
 {bone1Name=dyn_skirt_front_l_02_01, bone2Name=dyn_skirt_front_l_02_MSTR, swing1LimitDegrees=6, swing2LimitDegrees=2, twistLimitDegrees=2}
 {bone1Name=dyn_skirt_front_l_02_02, bone2Name=dyn_skirt_front_l_02_01, swing1LimitDegrees=6, swing2LimitDegrees=2, twistLimitDegrees=2}
 {bone1Name=dyn_skirt_front_f, bone2Name=dyn_skirt_front_MSTR, swing1LimitDegrees=25, swing2LimitDegrees=2, twistLimitDegrees=2}
 {bone1Name=dyn_skirt_front_f1, bone2Name=dyn_skirt_front_f, swing1LimitDegrees=25, swing2LimitDegrees=2, twistLimitDegrees=2}
 {bone1Name=dyn_skirt_front_r_01_01, bone2Name=dyn_skirt_front_r_01_MSTR, swing1LimitDegrees=6, swing2LimitDegrees=2, twistLimitDegrees=2}
 {bone1Name=dyn_skirt_front_r_01_02, bone2Name=dyn_skirt_front_r_01_01, swing1LimitDegrees=6, swing2LimitDegrees=2, twistLimitDegrees=2}
 {bone1Name=dyn_skirt_front_r_02_01, bone2Name=dyn_skirt_front_r_02_MSTR, swing1LimitDegrees=6, swing2LimitDegrees=2, twistLimitDegrees=2}
 {bone1Name=dyn_skirt_front_r_02_02, bone2Name=dyn_skirt_front_r_02_01, swing1LimitDegrees=6, swing2LimitDegrees=2, twistLimitDegrees=2}
 {bone1Name=dyn_skirt_back_l, bone2Name=dyn_skirt_back_l_MSTR, swing1LimitDegrees=25, swing2LimitDegrees=2, twistLimitDegrees=2}
 {bone1Name=dyn_skirt_back_l_01, bone2Name=dyn_skirt_back_l, swing1LimitDegrees=25, swing2LimitDegrees=2, twistLimitDegrees=2}
 {bone1Name=dyn_skirt_back_r, bone2Name=dyn_skirt_back_r_MSTR, swing1LimitDegrees=25, swing2LimitDegrees=2, twistLimitDegrees=2}
 {bone1Name=dyn_skirt_back_r_01, bone2Name=dyn_skirt_back_r, swing1LimitDegrees=25, swing2LimitDegrees=2, twistLimitDegrees=2}
 {bone1Name=dyn_skirt_r_fr_01, bone2Name=dyn_skirt_r_fr, swing1LimitDegrees=12, swing2LimitDegrees=2, twistLimitDegrees=2}
 {bone1Name=dyn_skirt_r_fr, bone2Name=dyn_skirt_r_fr_MSTR, swing1LimitDegrees=12, swing2LimitDegrees=2, twistLimitDegrees=2}
 {bone1Name=dyn_skirt_r_back_01, bone2Name=dyn_skirt_r_back, swing1LimitDegrees=25, swing2LimitDegrees=2, twistLimitDegrees=2}
 {bone1Name=dyn_skirt_r_back, bone2Name=dyn_skirt_r_back_MSTR, swing1LimitDegrees=25, swing2LimitDegrees=2, twistLimitDegrees=2}
 {bone1Name=dyn_skirt_l_fr_01, bone2Name=dyn_skirt_l_fr, swing1LimitDegrees=12, swing2LimitDegrees=2, twistLimitDegrees=2}
 {bone1Name=dyn_skirt_l_fr, bone2Name=dyn_skirt_l_fr_MSTR, swing1LimitDegrees=12, swing2LimitDegrees=2, twistLimitDegrees=2}
 {
  bone1Name=calf_r
  bone2Name=thigh_r
  swing1LimitDegrees=45
  swing1Motion=Free
  swing2LimitDegrees=45
  swing2Motion=Free
  twistLimitDegrees=45
  twistMotion=Free
 }
 {
  bone1Name=foot_l
  bone2Name=calf_l
  swing1LimitDegrees=45
  swing1Motion=Free
  swing2LimitDegrees=45
  swing2Motion=Free
  twistLimitDegrees=45
  twistMotion=Free
 }
 {
  bone1Name=calf_l
  bone2Name=thigh_l
  swing1LimitDegrees=45
  swing1Motion=Free
  swing2LimitDegrees=45
  swing2Motion=Free
  twistLimitDegrees=45
  twistMotion=Free
 }
 {
  bone1Name=foot_r
  bone2Name=calf_r
  swing1LimitDegrees=45
  swing1Motion=Free
  swing2LimitDegrees=45
  swing2Motion=Free
  twistLimitDegrees=45
  twistMotion=Free
 }
 {
  bone1Name=dyn_skirt_front_r_01_MSTR
  bone2Name=dyn_belt_r
  swing1LimitDegrees=45
  swing1Motion=Free
  swing2LimitDegrees=45
  swing2Motion=Free
  twistLimitDegrees=45
  twistMotion=Free
 }
 {
  bone1Name=dyn_skirt_front_r_02_MSTR
  bone2Name=dyn_belt_r
  swing1LimitDegrees=45
  swing1Motion=Free
  swing2LimitDegrees=45
  swing2Motion=Free
  twistLimitDegrees=45
  twistMotion=Free
 }
 {
  bone1Name=dyn_belt_r
  bone2Name=pelvis
  swing1LimitDegrees=45
  swing1Motion=Free
  swing2LimitDegrees=45
  swing2Motion=Free
  twistLimitDegrees=45
  twistMotion=Free
 }
 {
  bone1Name=dyn_skirt_front_MSTR
  bone2Name=dyn_belt_mid
  swing1LimitDegrees=45
  swing1Motion=Free
  swing2LimitDegrees=45
  swing2Motion=Free
  twistLimitDegrees=45
  twistMotion=Free
 }
 {
  bone1Name=dyn_belt_mid
  bone2Name=pelvis
  swing1LimitDegrees=45
  swing1Motion=Free
  swing2LimitDegrees=45
  swing2Motion=Free
  twistLimitDegrees=45
  twistMotion=Free
 }
 {
  bone1Name=dyn_skirt_front_l_01_MSTR
  bone2Name=dyn_belt_l
  swing1LimitDegrees=45
  swing1Motion=Free
  swing2LimitDegrees=45
  swing2Motion=Free
  twistLimitDegrees=45
  twistMotion=Free
 }
 {
  bone1Name=dyn_skirt_front_l_02_MSTR
  bone2Name=dyn_belt_l
  swing1LimitDegrees=45
  swing1Motion=Free
  swing2LimitDegrees=45
  swing2Motion=Free
  twistLimitDegrees=45
  twistMotion=Free
 }
 {
  bone1Name=dyn_belt_l
  bone2Name=pelvis
  swing1LimitDegrees=45
  swing1Motion=Free
  swing2LimitDegrees=45
  swing2Motion=Free
  twistLimitDegrees=45
  twistMotion=Free
 }
 {
  bone1Name=dyn_skirt_l_back_MSTR
  bone2Name=pelvis
  swing1LimitDegrees=45
  swing1Motion=Free
  swing2LimitDegrees=45
  swing2Motion=Free
  twistLimitDegrees=45
  twistMotion=Free
 }
 {bone1Name=dyn_skirt_l_back, bone2Name=dyn_skirt_l_back_MSTR, swing1LimitDegrees=25, swing2LimitDegrees=2, twistLimitDegrees=2}
 {bone1Name=dyn_skirt_l_back_01, bone2Name=dyn_skirt_l_back, swing1LimitDegrees=25, swing2LimitDegrees=2, twistLimitDegrees=2}
]
