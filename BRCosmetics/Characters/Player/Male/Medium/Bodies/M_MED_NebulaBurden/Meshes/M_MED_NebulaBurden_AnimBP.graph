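# Post-process style animation Blueprint for the NebulaBurden medium male cosmetic body, handling its skirt and floating stone secondary motion.
# Parent mesh pose passes through a state-fed Control Rig, additive front-skirt offsets, pelvis-space RigidBody simulation and additive dyn_stone bone transforms.
# The event graph converts player and melee-weapon states to floats, and computes skirt alpha and sine-driven rock offset vectors.
# Simulation gravity follows skydive-aware override, or standard gravity with long backblings.
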AnimBlueprint /BRCosmetics/Characters/Player/Male/Medium/Bodies/M_MED_NebulaBurden/Meshes/M_MED_NebulaBurden_AnimBP
target_skeleton=&/M_MED_NebulaBurden_Skeleton
graphs [
 AnimGraph {nodes [
   n1 = Control Rig!(
    Source=Copy Pose From Mesh() {bUseAttachedParent=true}
    Alpha
    skydiving floating:double=Get skydive floating() -> (skydive floating:double)
    sky diving parachute:double=Get sky dive parachute() -> (sky dive parachute:double)
    emoting:double=Get emoting() -> (emoting:double)
    is holding melee:double=Get is holding melee() -> (is holding melee:double)
    not crouching:double=Get not crouching() -> (not crouching:double)
    dbno:double=Get dbno() -> (dbno:double)
    playing melee animation:double=Get playing melee animation() -> (playing melee animation:double)
   ) -> (Pose) {
    alphaBoolBlend {blendOption=Linear}
    controlRigAssetReference {blueprintRigClass=&/M_MED_NebulaBurden_CtrlRig.M_MED_NebulaBurden_CtrlRig_C}
   }
   n2 = "Transform (Modify) Bone - Bone: dyn_skirt_front_l_01_MSTR"(ComponentPose=Local To Component(LocalPose=@n1), Alpha=1, Translation="0,-1,-2", Rotation="0,0,20", Scale="1,1,1") {
    rotationMode=BMM_Additive
    rotationSpace=BCS_BoneSpace
    translationMode=BMM_Additive
    translationSpace=BCS_BoneSpace
    alphaBoolBlend {blendOption=Linear}
    boneToModify {boneName=dyn_skirt_front_l_01_MSTR}
    rotation {roll=20}
    translation {y=-1, z=-2}
   }
   n3 = "Transform (Modify) Bone - Bone: dyn_skirt_front_r_01_MSTR"(ComponentPose=@n2, Alpha=1, Translation="0,-1,2", Rotation="0,0,-20", Scale="1,1,1") {
    rotationMode=BMM_Additive
    rotationSpace=BCS_BoneSpace
    translationMode=BMM_Additive
    translationSpace=BCS_BoneSpace
    alphaBoolBlend {blendOption=Linear}
    boneToModify {boneName=dyn_skirt_front_r_01_MSTR}
    rotation {roll=-20}
    translation {y=-1, z=2}
   }
   n4 = "Transform (Modify) Bone - Bone: dyn_skirt_front_l_02_MSTR"(ComponentPose=@n3, Alpha=1, Translation="0,0,-1", Rotation="0,0,10", Scale="1,1,1") {
    rotationMode=BMM_Additive
    rotationSpace=BCS_BoneSpace
    translationMode=BMM_Additive
    translationSpace=BCS_BoneSpace
    alphaBoolBlend {blendOption=Linear}
    boneToModify {boneName=dyn_skirt_front_l_02_MSTR}
    rotation {roll=10}
    translation {z=-1}
   }
   n5 = "Transform (Modify) Bone - Bone: dyn_skirt_front_r_02_MSTR"(ComponentPose=@n4, Alpha=1, Translation="0,0,1", Rotation="0,0,-10", Scale="1,1,1") {
    rotationMode=BMM_Additive
    rotationSpace=BCS_BoneSpace
    translationMode=BMM_Additive
    translationSpace=BCS_BoneSpace
    alphaBoolBlend {blendOption=Linear}
    boneToModify {boneName=dyn_skirt_front_r_02_MSTR}
    rotation {roll=-10}
    translation {z=1}
   }
   n6 = RigidBody!(ComponentPose=@n5, Alpha, OverrideWorldGravity:Vector=Get Override World Gravity() -> (Override World Gravity:Vector), ExternalForce) -> (Pose) {
    bClampLinearTranslationLimitToRefPose=true
    bOverrideWorldGravity=true
    overridePhysicsAsset=&/M_MED_NebulaBurden_Physics
    simulationSpace=BaseBoneSpace
    alphaBoolBlend {blendOption=Linear}
    baseBoneRef {boneName=pelvis}
    componentAppliedLinearAccClamp=<3000 3000 3000>
    componentLinearAccScale=<0.4 0.3 0.3>
    componentLinearVelScale=<0.4 0.3 0.3>
    simSpaceSettings {
     maxAngularAcceleration=1000
     maxAngularVelocity=3
     maxLinearAcceleration=700
     maxLinearVelocity=500
     externalAngularVelocity {z=0.1}
     externalLinearDragV=<0.2 0.2>
    }
   }
   n7 = "Transform (Modify) Bone - Bone: dyn_stone_03"(ComponentPose=@n6, Alpha=1, Translation="0,0,0", Rotation="0,0,0", Scale="1,1,1") {
    translationMode=BMM_Additive
    translationSpace=BCS_BoneSpace
    alphaBoolBlend {blendOption=Linear}
    alphaScaleBiasClamp {bInterpResult=true}
    boneToModify {boneName=dyn_stone_03}
   }
   n8 = "Transform (Modify) Bone - Bone: dyn_stone_04"(ComponentPose=@n7, Alpha=1, Translation="0,0,0", Rotation="0,0,0", Scale="1,1,1") {
    translationMode=BMM_Additive
    translationSpace=BCS_BoneSpace
    alphaBoolBlend {blendOption=Linear}
    alphaScaleBiasClamp {bInterpResult=true}
    boneToModify {boneName=dyn_stone_04}
   }
   n9 = "Transform (Modify) Bone - Bone: dyn_stone_05"(ComponentPose=@n8, Alpha=1, Translation="0,0,0", Rotation="0,0,0", Scale="1,1,1") {
    translationMode=BMM_Additive
    translationSpace=BCS_BoneSpace
    alphaBoolBlend {blendOption=Linear}
    alphaScaleBiasClamp {bInterpResult=true}
    boneToModify {boneName=dyn_stone_05}
   }
   n10 = "Transform (Modify) Bone - Bone: dyn_stone_06"(ComponentPose=@n9, Alpha=1, Translation="0,0,0", Rotation="0,0,0", Scale="1,1,1") {
    translationMode=BMM_Additive
    translationSpace=BCS_BoneSpace
    alphaBoolBlend {blendOption=Linear}
    alphaScaleBiasClamp {bInterpResult=true}
    boneToModify {boneName=dyn_stone_06}
   }
   n11 = "Transform (Modify) Bone - Bone: dyn_stone_01"(ComponentPose=@n10, Alpha=1, Translation="0,0,0", Rotation="0,0,0", Scale="1,1,1") {
    translationMode=BMM_Additive
    translationSpace=BCS_BoneSpace
    alphaBoolBlend {blendOption=Linear}
    alphaScaleBiasClamp {bInterpResult=true}
    boneToModify {boneName=dyn_stone_01}
   }
   n12 = "Transform (Modify) Bone - Bone: dyn_stone_02"(ComponentPose=@n11, Alpha=1, Translation="0,0,0", Rotation="0,0,0", Scale="1,1,1") {
    translationMode=BMM_Additive
    translationSpace=BCS_BoneSpace
    alphaBoolBlend {blendOption=Linear}
    alphaScaleBiasClamp {bInterpResult=true}
    boneToModify {boneName=dyn_stone_02}
   }
   n13 = "Transform (Modify) Bone - Bone: dyn_stone_07"(ComponentPose=@n12, Alpha=1, Translation="0,0,0", Rotation="0,0,0", Scale="1,1,1") {
    translationMode=BMM_Additive
    translationSpace=BCS_BoneSpace
    alphaBoolBlend {blendOption=Linear}
    alphaScaleBiasClamp {bInterpResult=true}
    boneToModify {boneName=dyn_stone_07}
   }
   n14 = "Transform (Modify) Bone - Bone: dyn_stone_08"(ComponentPose=@n13, Alpha=1, Translation="0,0,0", Rotation="0,0,0", Scale="1,1,1") {
    translationMode=BMM_Additive
    translationSpace=BCS_BoneSpace
    alphaBoolBlend {blendOption=Linear}
    alphaScaleBiasClamp {bInterpResult=true}
    boneToModify {boneName=dyn_stone_08}
   }
   n15 = "Transform (Modify) Bone - Bone: dyn_stone_09"(ComponentPose=@n14, Alpha=1, Translation="0,0,0", Rotation="0,0,0", Scale="1,1,1") {
    translationMode=BMM_Additive
    translationSpace=BCS_BoneSpace
    alphaBoolBlend {blendOption=Linear}
    alphaScaleBiasClamp {bInterpResult=true}
    boneToModify {boneName=dyn_stone_09}
   }
   n16 = "Transform (Modify) Bone - Bone: dyn_stone_10"(ComponentPose=@n15, Alpha=1, Translation="0,0,0", Rotation="0,0,0", Scale="1,1,1") {
    translationMode=BMM_Additive
    translationSpace=BCS_BoneSpace
    alphaBoolBlend {blendOption=Linear}
    alphaScaleBiasClamp {bInterpResult=true}
    boneToModify {boneName=dyn_stone_10}
   }
   n17 = "Transform (Modify) Bone - Bone: dyn_stone_11"(ComponentPose=@n16, Alpha=1, Translation="0,0,0", Rotation="0,0,0", Scale="1,1,1") {
    translationMode=BMM_Additive
    translationSpace=BCS_BoneSpace
    alphaBoolBlend {blendOption=Linear}
    alphaScaleBiasClamp {bInterpResult=true}
    boneToModify {boneName=dyn_stone_11}
   }
   n18 = "Transform (Modify) Bone - Bone: dyn_stone_12"(ComponentPose=@n17, Alpha=1, Translation="0,0,0", Rotation="0,0,0", Scale="1,1,1") {
    translationMode=BMM_Additive
    translationSpace=BCS_BoneSpace
    alphaBoolBlend {blendOption=Linear}
    alphaScaleBiasClamp {bInterpResult=true}
    boneToModify {boneName=dyn_stone_12}
   }
   n19 = "Transform (Modify) Bone - Bone: dyn_stone_13"(ComponentPose=@n18, Alpha=1, Translation="0,0,0", Rotation="0,0,0", Scale="1,1,1") {
    translationMode=BMM_Additive
    translationSpace=BCS_BoneSpace
    alphaBoolBlend {blendOption=Linear}
    alphaScaleBiasClamp {bInterpResult=true}
    boneToModify {boneName=dyn_stone_13}
   }
   n20 = "Transform (Modify) Bone - Bone: dyn_stone_14"(ComponentPose=@n19, Alpha=1, Translation="0,0,0", Rotation="0,0,0", Scale="1,1,1") {
    translationMode=BMM_Additive
    translationSpace=BCS_BoneSpace
    alphaBoolBlend {blendOption=Linear}
    alphaScaleBiasClamp {bInterpResult=true}
    boneToModify {boneName=dyn_stone_14}
   }
   Output Pose(Result=Component To Local(ComponentPose=@n20)) {layerGroup=None}
  ]}
 EventGraph {nodes [
   n1 = Event BlueprintUpdateAnimation()
   > Set Delta Time X(Delta Time X=@n1.DeltaTimeX)
   > n2 = Sequence!(execute) -> (then_0, then_1, then_2:Exec, then_3:Exec)
   n3 = Get bIsSkydiveFloating()
   > Set skydive floating(skydive floating:double=FInterpTo(Current=Get skydive floating(), Target="To Float (Boolean)"(InBool=@n3), DeltaTime=Get Delta Time X(), InterpSpeed=20)) -> (Output_Get:double)
   n4 = "To Float (Boolean)"(InBool=Get bIsParachuteOpened())
   > Set sky dive parachute(sky dive parachute:double=@n4) -> (Output_Get:double)
   n5 = "To Float (Boolean)"(InBool=Get bIsPlayingEmote())
   > Set emoting(emoting:double=@n5) -> (Output_Get:double)
   n6 = Get bIsCrouching()
   > Set not crouching(not crouching:double="To Float (Boolean)"(InBool=NOT Boolean(A=@n6))) -> (Output_Get:double)
   > Set dbno(dbno:double="To Float (Boolean)"(InBool=Get bIsDBNO())) -> (Output_Get:double)
   > Set playing melee animation(playing melee animation:double="To Float (Boolean)"(InBool=Get bIsPlayingMeleeAnim())) -> (Output_Get:double)
   n7 = Sequence!(execute=@n2.then_1) -> (then_0, then_1)
   > n8 = Gravity Override State Function DEPRECATED(CharacterPartAnimInstance=Self-Reference(), GameTimeInSeconds=GetGameTimeInSeconds(), GravityOverrideParams=Get Gravity Override Params())
   > Set Override World Gravity(Override World Gravity:Vector=SelectVector(A="0,0,-980", B=@n8, bPickA=CheckForBackbling().bIsBackblingLong)) -> (Output_Get:Vector)
   n9 = Cast To FortPlayerAnimInstance(execute=@n7.then_1, Object=GetAnimInstance(self=GetOwnerSkeletalMeshComponent()))
   > Set is holding melee(is holding melee:double=0) -> (Output_Get:double)
   > n10 = Switch on EFortWeaponCoreAnimation!(execute, Selection:EFortWeaponCoreAnimation Enum=Get WeaponCoreAnim(self=@n9)) -> (
    Melee:Exec
    Pistol:Exec
    Shotgun:Exec
    PaperBlueprint:Exec
    Rifle:Exec
    MeleeOneHand:Exec
    MachinePistol:Exec
    RocketLauncher:Exec
    GrenadeLauncher:Exec
    GoingCommando:Exec
    AssaultRifle:Exec
    TacticalShotgun:Exec
    SniperRifle:Exec
    TrapPlacement:Exec
    ShoulderLauncher:Exec
    AbilityDecoTool:Exec
    Crossbow:Exec
    C4:Exec
    RemoteControl:Exec
    DualWield:Exec
    AR_BullPup:Exec
    AR_ForwardGrip:Exec
    MedPackPaddles:Exec
    SMG_P90:Exec
    AR_DrumGun:Exec
    Consumable_Small:Exec
    Consumable_Large:Exec
    Balloon:Exec
    MountedTurret:Exec
    CreativeTool:Exec
    ExplosiveBow:Exec
    AshtonIndigo:Exec
    AshtonChicago:Exec
    MeleeDualWield:Exec
    Unarmed:Exec
   )
   Set is holding melee(execute=[@n10.Melee, @n10.MeleeDualWield, @n10.MeleeOneHand], is holding melee=1)
   n11 = Branch(execute=@n2.then_3, Condition=Get bIsSkyDiving())
   > Set Skirt Alpha(Skirt Alpha:double=0.5) -> (Output_Get:double)
   Set Skirt Alpha(execute=@n11.else, Skirt Alpha=1)
   n12 = float * float(A=GetGameTimeInSeconds(), B=1.2) {is=title K2Node_PromotableOperator}
   n13 = float * float(A="Sin (Radians)"(A=float * float(A=1.25, B=@n12)), B=0.8) {is=title K2Node_CommutativeAssociativeBinaryOperator}
   n14 = float * float(A="Sin (Radians)"(A=float * float(A=2, B=@n12)), B=0.5)
   n15 = float * float(A="Sin (Radians)"(A=float * float(A=1.5, B=@n12)), B=0.4)
   Set Floating Rocks 01(execute=@n2.then_2, Floating Rocks 01:Vector=MakeVector(X=@n15, Y=@n14, Z=@n13)) -> (Output_Get:Vector)
   > Set Floating Rocks 02(Floating Rocks 02:Vector=MakeVector(X=float * float(A=@n15, B=-2) {is=title K2Node_PromotableOperator}, Y=float * float(A=@n14, B=-2), Z=float * float(A=@n13, B=-2))) -> (Output_Get:Vector)
   > Set Floating Rocks 03(Floating Rocks 03:Vector=MakeVector(X=float * float(A=@n15, B=1.5), Y=float * float(A=@n14, B=-2), Z=float * float(A=@n13, B=-0.8))) -> (Output_Get:Vector)
   FInterpTo(Current=Get sky dive parachute(), Target=@n4, DeltaTime=Get Delta Time X(), InterpSpeed=20)
   FInterpTo(Current=Get emoting(), Target=@n5, DeltaTime=Get Delta Time X(), InterpSpeed=20)
  ]}
]
variables [
 crouching {type=number}
 dbno {type=number}
 Delta Time X {type=number}
 emoting {type=number}
 Floating Rocks 01 {type=Vector}
 Floating Rocks 02 {type=Vector}
 Floating Rocks 03 {type=Vector}
 Game Time in Seconds {type=number}
 Gravity Override Params {
  type=GravityOverrideParamsStruct
  default {
   skydive_DivingDown {gravityOverride=<-3000 -500 -500>}
   skydive_FloatingDown {gravityOverride=<0 -700 -1500>}
   skydive_GoingUp {gravityOverride=<0 300 -1500>}
   skydive_Parachuting {gravityOverride=<-1500 -2000 -980>}
  }
 }
 is holding melee {type=number}
 jumping {type=number}
 not crouching {type=number}
 Override World Gravity {type=Vector}
 playing melee animation {type=number}
 Skirt Alpha {type=number}
 sky dive parachute {type=number}
 skydive floating {type=number}
]
